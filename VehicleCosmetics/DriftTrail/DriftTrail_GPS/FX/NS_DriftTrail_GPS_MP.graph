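# Niagara vehicle-cosmetic system for the GPS drift trail, drawing effects at two wheels, apparently the rear pair.
# Two velocity-pushed ribbons with arrow parameters trail behind, alongside growing, fading sprites and two wheel-wrap emitters with rotating meshes, pins and rings.
# Emitters follow skeletal mesh sockets, most adding a user offset; FN_DisableTireFX masks per wheel and VC_WheelWrapScale sizes wraps.
# Painted-variant overrides and bIsPainted switch colors between paint and defaults.
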NiagaraSystem /VehicleCosmetics/DriftTrail/DriftTrail_GPS/FX/NS_DriftTrail_GPS_MP
dependencies {
 usedDataInterfaces [&NiagaraDataInterfaceCurve, &NiagaraDataInterfaceArrayFloat3, &NiagaraDataInterfaceSkeletalMesh]
 usedDynamicInputs []
 usedModules [
  &self:SystemSpawnScript.SystemScriptSource.SystemScriptGraph.NiagaraNodeAssignment_0.SetVariables_#1
  &FN_UnpackVehicleFlags
  /VehicleCosmetics/FX/Niagara/Modules/VC_SetVehicleSpeedAttribs
  &SystemState
 ]
}
emitters {
 ConstantParticles {
  emitterUpdateScript [
   &EmitterState(Inactive Response=::NewEnumerator1, Life Cycle Mode=::NewEnumerator1, UseLoopDelay=false)
   &SpawnRate(SpawnRate=6, Use Spawn Probability=false)
  ]
  particleSpawnScript [
   &VC_PaintedVariantsOverride(
    Black=<rgba 0.0003 0.0003 0 1>
    BurntSienna=<rgba 0.8228 0.2122 0.1195 1>
    Cobalt=<rgba 0.0278 0.0628 0.6 2>
    Crimson=<rgba 0.45 0.0208 0.0199 1>
    DefaultColor=&MakeCustomLinearColorFromBool(Bool=$User.bIsPainted, True Color=$User.PaintedColor, False Color=<rgba 0 0.4179 0.9743 1>)
    ForestGreen=<rgba 0.0612 0.7305 0.0612 1>
    Grey=<rgba 0.5841 0.5841 0.5906 1>
    Lime=<rgba 0.3154 0.6 0.0398 1>
    Orange=<rgba 0.6 0.1541 0.0166 1>
    Pink=<rgba 0.4 0.0822 0.3796 1>
    Purple=<rgba 0.5089 0.0331 0.9216 1>
    Saffron=<rgba 0.9823 0.9734 0.107 1>
    SkyBlue=<rgba 0.1301 0.7529 1 1>
   )
   &InitializeParticle(
    Color Mode=::NewEnumerator0
    Lifetime=/VehicleCosmetics/FX/Niagara/Modules/MergeEmitters/VC_FloatBySpawnGroup(
     SpawnGroup or ExecIndex:NiagaraInt32=0
     CustomInt:NiagaraInt32=0
     MaximumSpawnGroups:NiagaraInt32=2
     Float SG 0:NiagaraFloat=0.8
     Float SG 1:NiagaraFloat=0
     Float SG 2:NiagaraFloat=0
     Float SG 3:NiagaraFloat=0
     Float SG 4:NiagaraFloat=0
     Float SG 5:NiagaraFloat=0
    )
    Sprite Rotation Mode=::NewEnumerator2
    Sprite Size Mode=::NewEnumerator3
    Uniform Sprite Size=&VC_FloatBySpawnGroup(
     SpawnGroup or ExecIndex:NiagaraInt32=0
     CustomInt:NiagaraInt32=0
     MaximumSpawnGroups:NiagaraInt32=2
     Float SG 0:NiagaraFloat=40
     Float SG 1:NiagaraFloat=0
     Float SG 2:NiagaraFloat=0
     Float SG 3:NiagaraFloat=0
     Float SG 4:NiagaraFloat=0
     Float SG 5:NiagaraFloat=0
    )
    UsePositionOffset=false
    Write Material Random=true
   )
   &PartitionParticles()
   &SkeletalMeshLocation(
    Bone / Socket Index=&Add_Integer(A=$Particles.PartitionParticles.Partition, B=4)
    Bone / Socket Sampling Mode=::NewEnumerator5
    Interpolate Mesh Sampling=false
    MaskSpawnGroup=false
    Mesh Sampling Type=::NewEnumerator6
    Skeletal Mesh=$System.SkeletalMesh
   )
   &AddVelocity(Velocity=&VC_Mask_Vectors(
     FrontEnd Value:Vector3f=<-1500 0 0>
     blsFrontEndPreview:NiagaraBool=false
     SpeedMin Value:Vector3f=<-800 0 0>
     SpeedMax Value:Vector3f=<-1 0 0>
     Speed Alpha:NiagaraFloat=$System.Speed_Alpha
    )) {enabled=false}
  ]
  particleUpdateScript [
   &FN_DisableTireFX(
    Mask Index Attribute:NiagaraInt32=$Particles.PartitionParticles.Partition
    Mask Index FL=2
    Mask Index FR=3
    Mask Index RL=0
    Mask Index RR=1
   )
   &ParticleState()
   &ScaleSpriteSize(Uniform Curve Sprite Scale=DataInterface() {propertyValues {Curve {keys [{tangentMode=RCTM_User, time=0, value=0.1}, {tangentMode=RCTM_User, time=1, value=1.8}]}}})
   &SolveForcesAndVelocity(Clamp Velocity=false, Limit Acceleration=false)
   &ScaleColor(
    Scale Alpha=&FloatFromCurve(FloatCurve=DataInterface() {propertyValues {Curve {keys [{interpMode=RCIM_Cubic, tangentMode=RCTM_Break, time=0.3, value=1.5}, {interpMode=RCIM_Cubic, tangentMode=RCTM_Break, time=1, value=0}]}}})
    Scale RGB=&Multiply_VectorByFloat(
     Vector=<1 1 1>
     Float=&FloatFromCurve(FloatCurve=DataInterface() {propertyValues {Curve {keys [{interpMode=RCIM_Cubic, tangentMode=RCTM_Break, time=0, value=3.5}, {interpMode=RCIM_Cubic, tangentMode=RCTM_Break, time=0.8, value=1}]}}})
    )
    ScaleA=true
    ScaleRGB=true
   )
   &SkeletalMeshLocation(
    Bone / Socket Index=&Add_Integer(A=$Particles.PartitionParticles.Partition, B=4)
    Bone / Socket Sampling Mode=::NewEnumerator5
    MaskSpawnGroup=false
    Mesh Sampling Type=::NewEnumerator6
    Sampled Position Offset=&AddVectorToPosition(Position=<-80 0 -20>, Vector=$User.offset)
    Skeletal Mesh=$System.SkeletalMesh
   ) {module=SkeletalMeshLocation001}
  ]
  rendererClasses [&NiagaraSpriteRendererProperties]
  renderers [
   {rendererClass=&NiagaraSpriteRendererProperties}
   {rendererClass=&NiagaraSpriteRendererProperties, rendererIndex=1}
  ]
 }
 NE_VC_DriftTrail_Ribbon_Parent {
  emitterUpdateScript [
   &EmitterState()
   &VC_RateQualityHandler(Epic Spawn Rate=40, High Spawn Rate=35, Low Spawn Rate=25, Medium Spawn Rate=25, Use Custom Rate?=true)
   &VC_SpawnRibbon_v1(Number of Ribbons=2, Spawn Time=-1, SpawnRate=$Emitter.Spawn Rate)
  ]
  particleSpawnScript [
   &VC_PaintedVariantsOverride(
    Black=<rgba 0.0138 0.0138 0.015 1>
    BurntSienna=<rgba 0.5 0.1625 0.0965 1>
    Cobalt=<rgba 0.075 0.1875 0.75 1>
    Crimson=<rgba 0.8 0.056 0.0932 1>
    ForestGreen=<rgba 0.0809 0.8087 0.0809 1>
    Grey=<rgba 0.25 0.25 0.25 1>
    Lime=<rgba 0.3769 0.65 0.0474 1>
    Orange=<rgba 0.6 0.2381 0.0206 1>
    Pink=<rgba 0.7 0.2117 0.6937 1>
    Purple=<rgba 0.5297 0.047 0.65 1>
    R1 Black:LinearColor=<rgba 0.0482 0.0482 0.0497 1>
    R1 BurntSienna:LinearColor=<rgba 1 0.624 0.4125 1>
    R1 Cobalt:LinearColor=<rgba 0.2423 0.5647 1 1>
    R1 Crimson:LinearColor=<rgba 1 0.2 0.2267 1>
    R1 ForestGreen:LinearColor=<rgba 0.3613 1 0.3663 1>
    R1 Grey:LinearColor=<rgba 0.3955 0.3955 0.4 1>
    R1 Lime:LinearColor=<rgba 1 1 0.2195 1>
    R1 Orange:LinearColor=<rgba 1 0.7529 0.0723 1>
    R1 Pink:LinearColor=<rgba 1 0.4 1 1>
    R1 Purple:LinearColor=<rgba 0.7 0.1245 0.7 1>
    R1 Saffron:LinearColor=<rgba 1 1 0.2664 1>
    R1 SkyBlue:LinearColor=<rgba 0.2 1 1 1>
    R1 TitaniumWhite:LinearColor=<rgba 0.0152 0.0152 0.016 1>
    R2 Black:LinearColor=<rgba 0.0044 0.0044 0.0048 1>
    R2 BurntSienna:LinearColor=<rgba 0.27 0.0666 0.054 1>
    R2 Cobalt:LinearColor=<rgba 0.0163 0.0491 0.55 1>
    R2 Crimson:LinearColor=<rgba 0.6 0.045 0.0561 1>
    R2 ForestGreen:LinearColor=<rgba 0.0331 0.5333 0.0331 1>
    R2 Grey:LinearColor=<rgba 0.14 0.14 0.14 1>
    R2 Lime:LinearColor=<rgba 0.1833 0.4 0.0206 1>
    R2 Orange:LinearColor=<rgba 0.65 0.1329 0.0137 1>
    R2 Pink:LinearColor=<rgba 0.5 0.0807 0.4781 1>
    R2 Purple:LinearColor=<rgba 0.1604 0.0138 0.6 1>
    R2 Saffron:LinearColor=<rgba 0.3 0.2962 0.0184 1>
    R2 SkyBlue:LinearColor=<rgba 0.031 0.2796 0.55 1>
    R3 Black:LinearColor=&Multiply_LinearColorByFloat(Linear Color=<rgba 0.0191 0.0191 0.0191 1>, Float=2.5)
    R3 BurntSienna:LinearColor=&Multiply_LinearColorByFloat(Linear Color=<rgba 0.3515 0.0497 0.0273 1>, Float=2.5)
    R3 Cobalt:LinearColor=&Multiply_LinearColorByFloat(Linear Color=<rgba 0.1 0.1389 1 1>, Float=1)
    R3 Crimson:LinearColor=&Multiply_LinearColorByFloat(Linear Color=<rgba 0.5 0.01 0.0182 1>, Float=3)
    R3 ForestGreen:LinearColor=&Multiply_LinearColorByFloat(Linear Color=<rgba 0.0144 0.2705 0.0144 1>, Float=3)
    Saffron=<rgba 0.6 0.6 0.0442 1>
    SkyBlue=<rgba 0.09 0.6528 0.9 1>
    TitaniumWhite=<rgba 0.1978 0.1978 0.2 1>
    Use R1 Painted Overrides=true
    Use R2 Painted Overrides=true
    Use R3 Painted Overrides=true
   )
   &VC_InitializeRibbon_v1(
    Color=$Particles.Color
    Custom Offset=true
    Enable Skeletal Position=true
    Lifetime=0.8
    Local Vector Offset=&AddVectorToPosition(Position=<-65 0 -20>, Vector=$User.offset)
    Offset Amount=1
    Ribbon Width=25
   )
   &AddVelocity(Velocity=<-1000 0 0>) {module=AddVelocity001}
   /SetVariables(Particles.ArrowWidth=23) {isSetParameters=true}
  ]
  particleUpdateScript [
   &FN_DisableTireFX(Mask Index Attribute:NiagaraInt32=$Particles.SpawnGroup, Mask Index FL=2, Mask Index FR=3, Mask Index RL=0, Mask Index RR=1)
   &ParticleState()
   &InheritSourceMovement(
    Applied Position Scale=&Multiply_VectorByFloat(
     Vector=<1 1 1>
     Float=&FloatFromCurve(FloatCurve=DataInterface() {propertyValues {Curve {keys [{interpMode=RCIM_Cubic, tangentMode=RCTM_Break, time=0.3, value=1}, {interpMode=RCIM_Cubic, tangentMode=RCTM_Break, time=1, value=0}]}}})
    )
    ApplyForce=false
    ApplyPosition=true
    ApplyPositionFalloff=false
   )
   &SolveForcesAndVelocity(Clamp Velocity=false, Limit Acceleration=false)
   &VC_UpdateRibbon_v1(Enable Skeletal Position=true)
   &DynamicMaterialParameters(
    Index 0 Linear Color:LinearColor=&MakeCustomLinearColorFromBool(Bool=$User.bIsPainted, True Color=$Particles.Color, False Color=<rgba 0.4 0.1194 0.0031 1>)
    Index 1 Param 1=&MakeCustomFloatFromBool(Bool=$User.bIsPainted, True Float=1, False Float=0)
    Index 1 Param 2=&VC_FloatBySpawnGroup(
     SpawnGroup or ExecIndex:NiagaraInt32=0
     CustomInt:NiagaraInt32=0
     MaximumSpawnGroups:NiagaraInt32=2
     Float SG 0:NiagaraFloat=0.5
     Float SG 1:NiagaraFloat=0
     Float SG 2:NiagaraFloat=0
     Float SG 3:NiagaraFloat=0
     Float SG 4:NiagaraFloat=0
     Float SG 5:NiagaraFloat=0
    )
    Index 1 Param 3=1
    Index 1 Param 4=1
    Index 2 Linear Color:LinearColor=&MakeCustomLinearColorFromBool(Bool=$User.bIsPainted, True Color=$Particles.R1 Color, False Color=<rgba 0.887 0.7647 0.0288 1>)
    Index 3 Linear Color:LinearColor=&MakeCustomLinearColorFromBool(Bool=$User.bIsPainted, True Color=$Particles.R2 Color, False Color=<rgba 0.0176 0.6308 0.9216 1>)
    Module.Index1Float0Write=true
    Module.Index1Float1Write=true
    Module.Index1Float2Write=true
    Module.Index1Float3Write=true
    Write Parameter Index 0=::NewEnumerator5
    Write Parameter Index 1=::NewEnumerator1
    Write Parameter Index 2=::NewEnumerator5
    Write Parameter Index 3=::NewEnumerator5
   )
   /SetVariables(
    Particles.ArrowColor:LinearColor=&MakeCustomLinearColorFromBool(Bool=$User.bIsPainted, True Color=$Particles.R3 Color, False Color=<rgba 3 0.1426 0.1426 1>)
    Particles.DynamicIndex_Arrow:Vector4f=&MakeVector4(X=1, Y=0, Z=0, W=0)
   ) {isSetParameters=true}
   &ScaleColor(
    Scale Alpha=&FloatFromCurve(FloatCurve=DataInterface() {propertyValues {Curve {keys [<at 0 -0.2>,<at 0.2 1>,<at 0.8 1>,<at 1 1>]}}})
    ScaleA=true
    ScaleRGB=true
   )
  ]
  rendererClasses [&NiagaraRibbonRendererProperties]
  renderers [
   {rendererClass=&NiagaraRibbonRendererProperties}
   {rendererClass=&NiagaraRibbonRendererProperties, rendererIndex=1}
  ]
 }
 NE_VC_DriftTrail_WheelWrap_Parent {
  emitterUpdateScript [
   &EmitterState(
    Inactive Response=::NewEnumerator1
    Life Cycle Mode=::NewEnumerator1
    Loop Behavior=::NewEnumerator1
    Loop Duration Mode=::NewEnumerator1
    UseLoopDelay=false
   )
   &SpawnBurst_Instantaneous(Spawn Count=2, Use Spawn Probability=false, UseLoopCountLimit=false)
   &SpawnBurst_Instantaneous(Spawn Count=2, Spawn Group=1, Use Spawn Probability=false, UseLoopCountLimit=false) {module=SpawnBurst_Instantaneous001}
   &SpawnBurst_Instantaneous(Spawn Count=2, Spawn Group=2, Use Spawn Probability=false, UseLoopCountLimit=false) {module=SpawnBurst_Instantaneous003}
   &SpawnBurst_Instantaneous(Spawn Count=2, Spawn Group=3, Use Spawn Probability=false, UseLoopCountLimit=false) {module=SpawnBurst_Instantaneous002}
  ]
  particleSpawnScript [
   &VC_PaintedVariantsOverride(
    Black=<rgba 0.0138 0.0138 0.015 1>
    BurntSienna=<rgba 0.5 0.1625 0.0965 1>
    Cobalt=<rgba 0.075 0.1875 0.75 1>
    Crimson=<rgba 0.7 0.0749 0.0749 1>
    ForestGreen=<rgba 0.0809 0.8087 0.0809 1>
    Grey=<rgba 0.25 0.25 0.25 1>
    Lime=<rgba 0.3769 0.65 0.0474 1>
    Orange=<rgba 0.6 0.2381 0.0206 1>
    Pink=<rgba 0.7 0.2117 0.6937 1>
    Purple=<rgba 0.5297 0.047 0.65 1>
    R1 Black:LinearColor=<rgba 0.0482 0.0482 0.0497 1>
    R1 BurntSienna:LinearColor=<rgba 1 0.624 0.4125 1>
    R1 Cobalt:LinearColor=<rgba 0.2423 0.5647 1 1>
    R1 Crimson:LinearColor=<rgba 1 0.2 0.2 1>
    R1 ForestGreen:LinearColor=<rgba 0.3613 1 0.3663 1>
    R1 Grey:LinearColor=<rgba 0.3955 0.3955 0.4 1>
    R1 Lime:LinearColor=<rgba 1 1 0.2195 1>
    R1 Orange:LinearColor=<rgba 1 0.7529 0.0723 1>
    R1 Pink:LinearColor=<rgba 1 0.4 1 1>
    R1 Purple:LinearColor=<rgba 0.7 0.1245 0.7 1>
    R1 Saffron:LinearColor=<rgba 1 1 0.2664 1>
    R1 SkyBlue:LinearColor=<rgba 0.2 1 1 1>
    R1 TitaniumWhite:LinearColor=<rgba 0.0152 0.0152 0.016 1>
    R2 Black:LinearColor=<rgba 0.0044 0.0044 0.0048 1>
    R2 BurntSienna:LinearColor=<rgba 0.27 0.0666 0.054 1>
    R2 Cobalt:LinearColor=<rgba 0.0163 0.0491 0.55 1>
    R2 Crimson:LinearColor=<rgba 0.6724 0.0296 0.0307 1>
    R2 ForestGreen:LinearColor=<rgba 0.0331 0.5333 0.0331 1>
    R2 Grey:LinearColor=<rgba 0.14 0.14 0.14 1>
    R2 Lime:LinearColor=<rgba 0.1833 0.4 0.0206 1>
    R2 Orange:LinearColor=<rgba 0.8388 0.1714 0.0176 1>
    R2 Pink:LinearColor=<rgba 0.5906 0.0953 0.5647 1>
    R2 Purple:LinearColor=<rgba 0.1871 0.0161 0.7 1>
    R2 Saffron:LinearColor=<rgba 0.3 0.2962 0.0184 1>
    R2 SkyBlue:LinearColor=<rgba 0.031 0.2796 0.55 1>
    R3 Black:LinearColor=<rgba 0.02 0.02 0.02 1>
    R3 BurntSienna:LinearColor=<rgba 0.2582 0.0319 0.0185 1>
    R3 Cobalt:LinearColor=<rgba 0.0094 0.0293 0.45 1>
    R3 Crimson:LinearColor=<rgba 0.2159 0.006 0.0065 1>
    R3 DefaultAlpha=0
    R3 DefaultColor:LinearColor=<rgba 1 1 1 0>
    R3 ForestGreen:LinearColor=<rgba 0.0052 0.0844 0.0056 1>
    R3 Grey:LinearColor=<rgba 0.0437 0.0437 0.0452 1>
    R3 Lime:LinearColor=<rgba 0.0529 0.1714 0.0086 1>
    R3 Orange:LinearColor=<rgba 0.3564 0.0369 0.007 1>
    R3 Pink:LinearColor=<rgba 0.2195 0.0222 0.1981 1>
    R3 Purple:LinearColor=<rgba 0.0545 0.004 0.2122 1>
    R3 Saffron:LinearColor=<rgba 0.0976 0.0953 0.0075 1>
    R3 SkyBlue:LinearColor=<rgba 0.007 0.0648 0.2016 1>
    R3 TitaniumWhite:LinearColor=<rgba 0.4953 0.4953 0.5 1>
    R4 Black:LinearColor=<rgba 0.1968 0.1968 0.2 1>
    R4 BurntSienna:LinearColor=<rgba 1 0.4942 0.3 1>
    R4 Cobalt:LinearColor=<rgba 0.3185 0.6939 1 1>
    R4 Crimson:LinearColor=<rgba 1 0.3663 0.3712 1>
    R4 ForestGreen:LinearColor=<rgba 0.1812 0.9131 0.1845 1>
    R4 Grey:LinearColor=<rgba 0.3961 0.3961 0.4 1>
    R4 Lime:LinearColor=<rgba 1 1 0.2195 1>
    R4 Orange:LinearColor=<rgba 1 0.8632 0.0887 1>
    R4 Pink:LinearColor=<rgba 0.9473 0.3515 0.9387 1>
    R4 Purple:LinearColor=<rgba 1 0.1559 1 1>
    R4 Saffron:LinearColor=<rgba 0.9047 0.8963 0.1329 1>
    R4 SkyBlue:LinearColor=<rgba 0.3968 0.9911 0.9911 1>
    R4 TitaniumWhite:LinearColor=<rgba 0.0152 0.0152 0.016 1>
    Saffron=<rgba 0.6 0.6 0.0442 1>
    SkyBlue=<rgba 0.09 0.6528 0.9 1>
    TitaniumWhite=<rgba 0.1978 0.1978 0.2 1>
    Use R1 Painted Overrides=true
    Use R2 Painted Overrides=true
    Use R3 Painted Overrides=true
    Use R4 Painted Overrides=true
    Use R5 Painted Overrides=true
   )
   /SetVariables(Particles.NewSpawnFlags:NiagaraInt32=&SpawnGroup()) {isSetParameters=true}
   &InitializeParticle(
    Color=$Particles.Color
    Lifetime=&VC_FloatBySpawnGroup(
     SpawnGroup or ExecIndex:NiagaraInt32=0
     CustomInt:NiagaraInt32=0
     MaximumSpawnGroups:NiagaraInt32=5
     Float SG 0:NiagaraFloat=$Emitter.CurrentLoopDuration
     Float SG 1:NiagaraFloat=$Emitter.CurrentLoopDuration
     Float SG 2:NiagaraFloat=$Emitter.CurrentLoopDuration
     Float SG 3:NiagaraFloat=$Emitter.CurrentLoopDuration
     Float SG 4:NiagaraFloat=1
     Float SG 5:NiagaraFloat=0
    )
    Mesh Scale=/VehicleCosmetics/FX/Niagara/Modules/MergeEmitters/VC_VectorBySpawnGroup(
     SpawnGroup or ExecIndex:NiagaraInt32=0
     CustomInt:NiagaraInt32=0
     MaximumSpawnGroups:NiagaraInt32=5
     Vector Group 0:Vector3f=<1 1 1>
     Vector Group 1:Vector3f=<0 0 0>
     Vector Group 2:Vector3f=<1 1 1>
     Vector Group 3:Vector3f=<0 0 0>
     Vector Group 4:Vector3f=<0 0 0>
     Vector Group 5:Vector3f=<0 0 0>
    )
    Mesh Scale Mode=::NewEnumerator4
    Position Mode=::NewEnumerator0
    Sprite Rotation Mode=::NewEnumerator2
    Sprite Size Mode=::NewEnumerator3
    Uniform Sprite Size=&VC_FloatBySpawnGroup(
     SpawnGroup or ExecIndex:NiagaraInt32=0
     CustomInt:NiagaraInt32=0
     MaximumSpawnGroups:NiagaraInt32=5
     Float SG 0:NiagaraFloat=0
     Float SG 1:NiagaraFloat=0
     Float SG 2:NiagaraFloat=50
     Float SG 3:NiagaraFloat=11
     Float SG 4:NiagaraFloat=60
     Float SG 5:NiagaraFloat=0
    )
    Write Material Random=true
   )
   &PartitionParticles()
   &InitialMeshOrientation(
    Mesh Orientation Mode=::NewEnumerator7
    Rotation=&VC_VectorBySpawnGroup(
     SpawnGroup or ExecIndex:NiagaraInt32=0
     CustomInt:NiagaraInt32=0
     MaximumSpawnGroups:NiagaraInt32=5
     Vector Group 0:Vector3f=<0 0 0>
     Vector Group 1:Vector3f=<0 0 -0.4>
     Vector Group 2:Vector3f=<0 0.4 0>
     Vector Group 3:Vector3f=<0 0 0>
     Vector Group 4:Vector3f=<0 0 0>
     Vector Group 5:Vector3f=<0 0 0>
    )
    Use Rotation Vector=true
   )
   &SkeletalMeshLocation(
    Bone / Socket Index=&Add_Integer(A=$Particles.PartitionParticles.Partition, B=4)
    Bone / Socket Sampling Mode=::NewEnumerator5
    Interpolate Mesh Sampling=false
    MaskSpawnGroup=false
    Mesh Sampling Type=::NewEnumerator6
    Sampled Position Offset Coordinate Space:ENiagaraSkelMeshTransforms=::NewEnumerator2
    Skeletal Mesh=$System.SkeletalMesh
   )
   /SetVariables(
    Particles.CircleColor:LinearColor=&MakeCustomLinearColorFromBool(Bool=$User.bIsPainted, True Color=$User.PaintedColor, False Color=<rgba 0 0.4179 0.9743 1>)
    Particles.PointColor:LinearColor=&MakeCustomLinearColorFromBool(Bool=$User.bIsPainted, True Color=$User.PaintedColor, False Color=<rgba 0.2878 0.7869 0.9828 1>)
    Particles.ScalePinLocal:Vector3f=&Multiply_VectorByFloat(Vector=<1 1 1>, Float=4)
   ) {isSetParameters=true}
   &ApplyInitialForces()
  ]
  particleUpdateScript [
   &FN_DisableTireFX(
    Mask Index Attribute:NiagaraInt32=$Particles.PartitionParticles.Partition
    Mask Index FL=2
    Mask Index FR=3
    Mask Index RL=0
    Mask Index RR=1
   )
   &VC_WheelWrapScale(Wheel Index Attribute:NiagaraInt32=&Add_Integer(A=$Particles.PartitionParticles.Partition, B=2))
   &ParticleState()
   &UpdateMeshOrientation(Rotation Coordinate Space=::NewEnumerator2, Rotation Vector=<0 1 0>) {enabled=false}
   &MeshRotationForce(Rotation=&MaskVectorBySpawnGroup(
     Masked Vector=&SelectVectorFromArray(
      Direct Array Index=$Particles.PartitionParticles.Partition
      Vector Selection Array=DataInterface() {propertyValues {FloatData [<0 0 0>, <0 0 0>]}}
     )
     Passthrough Vector=<0 0 0>
     Spawn Group Mask=1
    )) {enabled=false}
   &SkeletalMeshLocation(
    Bone / Socket Index=&Add_Integer(A=$Particles.PartitionParticles.Partition, B=4)
    Bone / Socket Sampling Mode=::NewEnumerator5
    MaskSpawnGroup=false
    Mesh Sampling Type=::NewEnumerator6
    Orientation Sampling=::NewEnumerator1
    Sampled Position Offset=&VC_VectorBySpawnGroup(
     SpawnGroup or ExecIndex:NiagaraInt32=0
     CustomInt:NiagaraInt32=0
     MaximumSpawnGroups:NiagaraInt32=5
     Vector Group 0:Vector3f=<0 0 0>
     Vector Group 1:Vector3f=&AddVectorToPosition(Position=<-80 0 -20>, Vector=$User.offset)
     Vector Group 2:Vector3f=&AddVectorToPosition(Position=<0 0 0>, Vector=$User.offset)
     Vector Group 3:Vector3f=&AddVectorToPosition(Position=<-80 0 -20>, Vector=$User.offset)
     Vector Group 4:Vector3f=&AddVectorToPosition(Position=<-80 0 -20>, Vector=$User.offset)
     Vector Group 5:Vector3f=<0 0 0>
    )
    Skeletal Mesh=$System.SkeletalMesh
   ) {module=SkeletalMeshLocation001}
   &UpdateMeshOrientation(
    Rotation Rate=1.8
    Rotation Vector=&MaskVectorBySpawnGroup(
     Masked Vector=&SelectVectorFromArray(
      Direct Array Index=$Particles.PartitionParticles.Partition
      Vector Selection Array=DataInterface() {propertyValues {FloatData [<0 0 0.3>, <0 0 -0.3>]}}
     )
     Passthrough Vector=<0 0 0>
     Spawn Group Mask=1
    )
   ) {module=UpdateMeshOrientation001}
   /SetVariables(
    Particles.DynamixPinMAps:Vector4f=&MakeVector4(X=1.5, Y=2, Z=2, W=0.05)
    Particles.DynamixPinMaps_02:Vector4f=&MakeVector4_FromLinearColor(LinearColor=$Particles.R3 Color)
    Particles.DynamixPinMaps_03:Vector4f=&MakeVector4_FromLinearColor(LinearColor=$Particles.R4 Color)
    Particles.ScaleCirlceColor:LinearColor=&MultiplyLinearColor(
     Linear Color=$Particles.CircleColor
     Scale Factor=&MakeVector4(
      X=1
      Y=1
      Z=1
      W=&FloatFromCurve(FloatCurve=DataInterface() {propertyValues {Curve {keys [{interpMode=RCIM_Cubic, tangentMode=RCTM_Break, time=0.3, value=1.5}, {interpMode=RCIM_Cubic, tangentMode=RCTM_Break, time=1, value=0}]}}})
     )
    )
    Particles.ScaleGlowRingColor:LinearColor=&MultiplyLinearColor(
     Linear Color=$Particles.CircleColor
     Scale Factor=&MakeVector4(
      X=4
      Y=4
      Z=4
      W=&FloatFromCurve(
       FloatCurve=DataInterface() {propertyValues {Curve {keys [{interpMode=RCIM_Cubic, tangentMode=RCTM_Break, time=0.3, value=1}, {interpMode=RCIM_Cubic, tangentMode=RCTM_Break, time=1, value=0}]}}}
       Scale Curve=0.8
      )
     )
    )
    Particles.ScalePointColor:LinearColor=&MultiplyLinearColor(Linear Color=$Particles.PointColor, Scale Factor=<5 5 6 1>)
   ) {isSetParameters=true}
   &DynamicMaterialParameters(
    Index 0 Linear Color:LinearColor=&MakeCustomLinearColorFromBool(Bool=$User.bIsPainted, True Color=$Particles.Color, False Color=<rgba 0.4 0.1194 0.0031 1>)
    Index 1 Param 1=&MakeCustomFloatFromBool(Bool=$User.bIsPainted, True Float=1, False Float=0)
    Index 1 Param 2=&VC_FloatBySpawnGroup(
     SpawnGroup or ExecIndex:NiagaraInt32=0
     CustomInt:NiagaraInt32=0
     MaximumSpawnGroups:NiagaraInt32=2
     Float SG 0:NiagaraFloat=0.5
     Float SG 1:NiagaraFloat=0
     Float SG 2:NiagaraFloat=0
     Float SG 3:NiagaraFloat=0
     Float SG 4:NiagaraFloat=0
     Float SG 5:NiagaraFloat=0
    )
    Index 1 Param 3=1
    Index 1 Param 4=1
    Index 2 Linear Color:LinearColor=&MakeCustomLinearColorFromBool(Bool=$User.bIsPainted, True Color=$Particles.R1 Color, False Color=<rgba 0.887 0.7647 0.0288 1>)
    Index 3 Linear Color:LinearColor=&MakeCustomLinearColorFromBool(Bool=$User.bIsPainted, True Color=$Particles.R2 Color, False Color=<rgba 0.0176 0.6308 0.9216 1>)
    Module.Index1Float0Write=true
    Module.Index1Float1Write=true
    Module.Index1Float2Write=true
    Module.Index1Float3Write=true
    Write Parameter Index 0=::NewEnumerator5
    Write Parameter Index 1=::NewEnumerator1
    Write Parameter Index 2=::NewEnumerator5
    Write Parameter Index 3=::NewEnumerator5
   )
   &SolveForcesAndVelocity(Clamp Velocity=false, Limit Acceleration=false)
   /SetVariables(
    Particles.ArrowColor_Tyre:LinearColor=&MakeCustomLinearColorFromBool(Bool=$User.bIsPainted, True Color=$Particles.R5 Color, False Color=<rgba 3 0.1426 0.1426 1>)
    Particles.DynamicIndexArrow_0:Vector4f=&MakeVector4(X=1, Y=0, Z=0, W=0)
    Particles.Vector 4:Vector4f=&MakeVector4(X=0, Y=0, Z=0, W=1)
   ) {isSetParameters=true}
  ]
  rendererClasses [&NiagaraMeshRendererProperties, &NiagaraSpriteRendererProperties]
  renderers [
   {rendererClass=&NiagaraMeshRendererProperties, rendererIndex=0}
   {rendererClass=&NiagaraSpriteRendererProperties, rendererIndex=1}
   {rendererClass=&NiagaraMeshRendererProperties, rendererIndex=2}
   {rendererClass=&NiagaraMeshRendererProperties, rendererIndex=3}
  ]
 }
 NE_VC_DriftTrail_WheelWrap_Parent001 {
  emitterUpdateScript [
   &EmitterState(
    Inactive Response=::NewEnumerator1
    Life Cycle Mode=::NewEnumerator1
    Loop Behavior=::NewEnumerator1
    Loop Delay=&SelectFloatFromArray(
     Array Sampling Mode=::NewEnumerator2
     Direct Array Index=?
     Interpolation=0
     Float Selection Array=DataInterface() {propertyValues {FloatData [0,0.5]}}
    )
    Loop Duration Mode=::NewEnumerator1
    UseLoopDelay=true
   )
   &SpawnBurst_Instantaneous(Spawn Count=2, Use Spawn Probability=false, UseLoopCountLimit=false)
  ]
  particleSpawnScript [
   &VC_PaintedVariantsOverride(
    Black=&Multiply_LinearColorByFloat(Linear Color=<rgba 0.0191 0.0191 0.0191 1>, Float=2.5)
    BurntSienna=&Multiply_LinearColorByFloat(Linear Color=<rgba 0.3515 0.0497 0.0273 1>, Float=2.5)
    Cobalt=&Multiply_LinearColorByFloat(Linear Color=<rgba 0.1 0.1389 1 1>, Float=1)
    Crimson=&Multiply_LinearColorByFloat(Linear Color=<rgba 0.5 0.01 0.0182 1>, Float=3)
    DefaultColor=&MakeCustomLinearColorFromBool(Bool=$User.bIsPainted, True Color=$User.PaintedColor, False Color=<rgba 0.9 0.0428 0.0428 1>)
    ForestGreen=&Multiply_LinearColorByFloat(Linear Color=<rgba 0.0144 0.2705 0.0144 1>, Float=3)
   )
   &InitializeParticle(
    Color=$Particles.Color
    Mesh Scale Mode=::NewEnumerator4
    Sprite Rotation Mode=::NewEnumerator2
    Sprite Size Mode=::NewEnumerator3
    UsePositionOffset=false
    Write Material Random=true
   )
   &PartitionParticles()
   &SkeletalMeshLocation(
    Bone / Socket Index=&Add_Integer(A=$Particles.PartitionParticles.Partition, B=4)
    Bone / Socket Sampling Mode=::NewEnumerator5
    Interpolate Mesh Sampling=false
    MaskSpawnGroup=false
    Mesh Sampling Type=::NewEnumerator6
    Skeletal Mesh=$System.SkeletalMesh
   )
   &InitialMeshOrientation(
    Mesh Orientation Mode=::NewEnumerator7
    Rotation=&SelectVectorFromArray(
     Direct Array Index=$Particles.PartitionParticles.Partition
     Vector Selection Array=DataInterface() {propertyValues {FloatData [<0 0 0>, <0 0.5 0>]}}
    )
    Use Rotation Vector=true
   )
  ]
  particleUpdateScript [
   &FN_DisableTireFX(
    Mask Index Attribute:NiagaraInt32=$Particles.PartitionParticles.Partition
    Mask Index FL=2
    Mask Index FR=3
    Mask Index RL=0
    Mask Index RR=1
   )
   &VC_WheelWrapScale(Wheel Index Attribute:NiagaraInt32=&Add_Integer(A=$Particles.PartitionParticles.Partition, B=2))
   &ParticleState(Kill Particles When Lifetime Has Elapsed=false)
   &UpdateMeshOrientation(Rotation Coordinate Space=::NewEnumerator2, Rotation Rate=0.85, Rotation Vector=<0 1 0>)
   &SkeletalMeshLocation(
    Bone / Socket Index=&Add_Integer(A=$Particles.PartitionParticles.Partition, B=4)
    Bone / Socket Sampling Mode=::NewEnumerator5
    MaskSpawnGroup=false
    Mesh Sampling Type=::NewEnumerator6
    Orientation Sampling=::NewEnumerator1
    Skeletal Mesh=$System.SkeletalMesh
   ) {module=SkeletalMeshLocation001}
   &SolveForcesAndVelocity(Clamp Velocity=false, Limit Acceleration=false)
   &DynamicMaterialParameters(
    Index 1 Param 4=1
    Module.Index1Float0Write=true
    Module.Index1Float1Write=true
    Module.Index1Float2Write=true
    Module.Index1Float3Write=true
    Write Parameter Index 0=::NewEnumerator0
    Write Parameter Index 1=::NewEnumerator1
   )
   /SetVariables(Particles.2ndTyre_01:Vector4f=&MakeVector4(
     X=0
     Y=0
     Z=0
     W=&FloatFromCurve(
      FloatCurve=DataInterface() {propertyValues {Curve {keys [<at 0 1>,<at 0.3 1>,<at 0.5 0>,<at 0.7 0>,<at 0.9 1>,<at 1 1>]}}}
      Scale Curve=&SelectFloatFromArray(Direct Array Index=$Particles.PartitionParticles.Partition, Float Selection Array=DataInterface() {propertyValues {FloatData [1,0]}})
     )
    )) {enabled=false, isSetParameters=true}
  ]
  rendererClasses [&NiagaraMeshRendererProperties]
  renderers [
   {rendererClass=&NiagaraMeshRendererProperties, rendererIndex=0}
   {rendererClass=&NiagaraMeshRendererProperties, rendererIndex=1}
  ]
 }
}
summary {
 systemName=NS_DriftTrail_GPS_MP
 emitters [
  {emitterName=NE_VC_DriftTrail_Ribbon_Parent, rendererClasses [&NiagaraRibbonRendererProperties]}
  {
   emitterName=NE_VC_DriftTrail_WheelWrap_Parent
   rendererClasses [&NiagaraMeshRendererProperties, &NiagaraSpriteRendererProperties]
  }
  {emitterName=ConstantParticles, rendererClasses [&NiagaraSpriteRendererProperties]}
  {emitterName=NE_VC_DriftTrail_WheelWrap_Parent001, rendererClasses [&NiagaraMeshRendererProperties]}
 ]
 userVariables [
  User.BackLine {type=LinearColor, description="", defaultValue {struct=LinearColor, value=<rgba 1 1 1 1>}}
  User.bIsFrontendPreview {type=NiagaraBool}
  User.bIsPainted {type=NiagaraBool}
  User.ColorVariantID {type=NiagaraInt32}
  User.LineaMiddle {type=LinearColor, description="", defaultValue {struct=LinearColor, value=<rgba 1 1 1 1>}}
  User.offset {type=Vector3f, description="offset from wheels(set through cosmetic template)", defaultValue {struct=Vector3f, value=<0 0 0>}}
  User.PaintedColor {type=LinearColor, description="", defaultValue {struct=LinearColor, value=<rgba 1 1 1 1>}}
  User.PaintedScalar {type=NiagaraFloat, defaultValue=1, description=""}
  User.SmallLine {type=LinearColor, description="", defaultValue {struct=LinearColor, value=<rgba 1 1 1 1>}}
  User.VehicleFlags {type=NiagaraInt32, defaultValue=806289408}
  User.Wheel_Radius_Array {
   type=NiagaraDataInterfaceArrayFloat
   defaultValue {
    struct=NiagaraExt_VariableValue_DataInterface
    value {
     dataInterface=&self:NiagaraDataInterfaceArrayFloat_1
     dataInterfaceClass=&NiagaraDataInterfaceArrayFloat
    }
   }
  }
  User.Wheel_Width_Array {
   type=NiagaraDataInterfaceArrayFloat
   defaultValue {
    struct=NiagaraExt_VariableValue_DataInterface
    value {
     dataInterface=&self:NiagaraDataInterfaceArrayFloat_0
     dataInterfaceClass=&NiagaraDataInterfaceArrayFloat
    }
   }
  }
  User.WheelTouchingArray {
   type=NiagaraDataInterfaceArrayFloat
   defaultValue {
    struct=NiagaraExt_VariableValue_DataInterface
    value {
     dataInterface=&self:NiagaraDataInterfaceArrayFloat_2
     dataInterfaceClass=&NiagaraDataInterfaceArrayFloat
    }
   }
  }
 ]
}
