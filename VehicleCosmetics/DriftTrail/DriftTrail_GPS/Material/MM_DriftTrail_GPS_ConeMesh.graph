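# Masked unlit material for Niagara mesh particles, used for the cone mesh in the GPS vehicle drift trail cosmetic.
# A view-vector and normal-map dot product is quantized with Ceil into stepped, toon-like shading that scales the emissive color.
# An IsPainted dynamic parameter switches between RGB-mask paint colors and a default texture/mask blend tinted by particle color.
# It also computes particle-times-texture alpha and a time-driven sine drift offset per axis.
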Material /VehicleCosmetics/DriftTrail/DriftTrail_GPS/Material/MM_DriftTrail_GPS_ConeMesh
graph {
 properties {
  bAutomaticallySetUsageInEditor=false
  bUsedWithNiagaraMeshParticles=true
  blendMode=BLEND_Masked
  shadingModel=MSM_Unlit
  usedShadingModels=Unlit
 }
 nodes [
  n1 = CameraVectorWS()
  n2 = Multiply(
   A=DotProduct(
    A=Transform(None=@n1) {transformSourceType=TRANSFORMSOURCE_World, transformType=TRANSFORM_Tangent}
    B=TextureSample() {samplerType=SAMPLERTYPE_Normal, texture=/VehicleCosmetics/DriftTrail/DriftTrail_GPS/Textures/T_VCVFX_EggNormals_N}.RGB
   )
   B=ScalarParameter() {defaultValue=3.5, parameterName=stepcount}
  )
  n3 = ScalarParameter() {defaultValue=1, parameterName=stepAmount}
  n4 = ConstantBiasScale(None=Ceil(None=Subtract(A=@n2, B=@n3))) {bias=2.5, scale=1}
  n5 = DynamicParameter() {parameterIndex=1, defaultValue=<rgba 0 1 1 1>, paramNames [IsPainted?, "", "", Param4]}
  n6 = TextureSample() {texture=/VehicleCosmetics/DriftTrail/DriftTrail_GPS/Textures/T_VCVFX_ConeMeshGPS_RGB}
  n7 = DynamicParameter() {parameterIndex=2, paramNames [PinOrangePainted, ., ., .]}
  n8 = Add(
   A=Add(A=Multiply(A=@n6.R, B=@n7.RGB), B=Multiply(A=@n6.G) {constB=0})
   B=Multiply(A=@n6.B, B=DynamicParameter() {parameterIndex=3, defaultValue=<rgba 0 1 1 1>, paramNames [DotMiddlePinPainted, ., ., .]}.RGB)
  )
  MaterialExpressionMeshEmitterDynamicParameter_1 = DynamicParameter() {paramNames [Param1,sang,vua,toi]}
  n9 = Add(
   A=Multiply(A=@MaterialExpressionMeshEmitterDynamicParameter_1.vua, B=@n6.R)
   B=Multiply(A=@MaterialExpressionMeshEmitterDynamicParameter_1.toi, B=@n6.G)
  )
  n10 = Multiply(A=@n6.B, B=@MaterialExpressionMeshEmitterDynamicParameter_1.sang)
  n11 = TextureSample() {texture=/VehicleCosmetics/DriftTrail/DriftTrail_GPS/Textures/T_VCVFX_ConeMeshGPS_D}
  n12 = ParticleColor()
  n13 = Multiply(A=LinearInterpolate(A=Add(A=@n9, B=@n10), B=@n11.RGB, Alpha=@MaterialExpressionMeshEmitterDynamicParameter_1.Param1), B=@n12.RGB)
  MaterialFunctionCall(Emissive=Multiply(A=Saturate(None=@n4), B=If(A=@n5.IsPainted?, "A > B"=@n8, "A == B"=@n8, "A < B"=@n13) {constB=0.5}) {constA=1}) {materialFunction=&MF_FX_EmissiveAdjustments}
  Multiply(A=@n12.A, B=@n11.A)
  n14 = Multiply(A=ScalarParameter() {defaultValue=1.5, parameterName=SpeedDB}, B=Time())
  n15 = Multiply(A=ScalarParameter() {defaultValue=1, parameterName=SpeedAB}, B=Time())
  n16 = Multiply(
   A=LinearInterpolate(A=Constant() {r=1}, B=Sine(None=@n14), Alpha=Sine(None=@n15))
   B=ScalarParameter() {defaultValue=1, parameterName=X_Drift_Mult}
  )
  n17 = Multiply(A=ScalarParameter() {defaultValue=1.5, parameterName=SpeedD}, B=Time())
  n18 = Multiply(A=ScalarParameter() {defaultValue=1, parameterName=SpeedA}, B=Time())
  n19 = Multiply(
   A=LinearInterpolate(A=Constant() {r=1}, B=Sine(None=@n17), Alpha=Sine(None=@n18))
   B=ScalarParameter() {defaultValue=1, parameterName=Y_Drift_Mult}
  )
  n20 = ScalarParameter() {defaultValue=1.5, parameterName=SpeedC}
  n21 = Time()
  n22 = ScalarParameter() {defaultValue=1, parameterName=SpeedB}
  n23 = Add(A=Time(), B=ScalarParameter() {defaultValue=1, parameterName=TimeOffset})
  n24 = LinearInterpolate(A=Sine(None=Multiply(A=@n20, B=@n21)), B=Constant() {r=-1}, Alpha=Cosine(None=Multiply(A=@n22, B=@n23)))
  n25 = ScalarParameter() {defaultValue=1, parameterName=Z_Drift_Mult}
  n26 = Multiply(
   A=AppendVector(A=AppendVector(A=@n16, B=@n19), B=Multiply(A=@n24, B=@n25))
   B=ScalarParameter() {defaultValue=5, parameterName=Uniform_Drift_Mult}
  )
  n27 = VectorParameter() {parameterName=WPO_OffsetB}
  Transform(None=Add(A=ComponentMask(None=@n26) {b=true, g=true, r=true}, B=ComponentMask(None=@n27.RGB) {b=true, g=true, r=true})) {transformSourceType=TRANSFORMSOURCE_Local}
 ]
}
