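# Unlit translucent master material for the GPS drift trail vehicle cosmetic, used on Niagara ribbons and mesh particles.
# Panning GPS textures, offset and tiled by parameters, act as channel masks for line colors from dynamic parameters, tinted or blended with particle color through MF_FX_EmissiveAdjustments.
# A Trail/Wheel scalar switches opacity between a U-faded trail and a rounded-sine banded wheel mask, scaled by particle alpha and Switch_TyreIntermit.
# The graph also holds a parallel color and opacity chain using IsPainted and center-line boost.
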Material /VehicleCosmetics/DriftTrail/DriftTrail_GPS/Material/MM_Drift_GPS_TrailMat
graph {
 properties {
  bAutomaticallySetUsageInEditor=false
  bUsedWithNiagaraMeshParticles=true
  bUsedWithNiagaraRibbons=true
  blendMode=BLEND_Translucent
  shadingModel=MSM_Unlit
  usedShadingModels=Unlit
 }
 nodes [
  DynamicParameter() {parameterIndex=1, defaultValue=<rgba 0 1 1 1>, paramNames [IsPainted?, OffsetFlechaX, CentralLineMulti, Switch_TyreIntermit]}
  n1 = MaterialFunctionCall(
   UVs=TextureCoordinate() {uTiling=0.55}
   Texture Scale=AppendVector(A=ScalarParameter() {defaultValue=1, parameterName=ArrowWheelU}, B=ScalarParameter() {defaultValue=1, parameterName=ArrowWheelV})
  ) {materialFunction=&ScaleUVsByCenter}
  n2 = MaterialFunctionCall(
   UVs=TextureCoordinate() {uTiling=1.5}
   Texture Scale=AppendVector(A=ScalarParameter() {defaultValue=1, parameterName=U}, B=ScalarParameter() {defaultValue=1, parameterName=V})
  ) {materialFunction=&ScaleUVsByCenter}
  n3 = ScalarParameter() {defaultValue=1, desc="1=X3Y1; 0=X1Y1", parameterName=SelectorTextCoord}
  n4 = DynamicParameter() {parameterIndex=1, defaultValue=<rgba 0 1 1 1>, paramNames [IsPainted?, OffsetFlechaX, CentralLineMulti, Switch_TyreIntermit]}
  n5 = ScalarParameter() {parameterName=OffsetFlechaY}
  n6 = Panner(
   Coordinate=Add(A=LinearInterpolate(A=@n1.UVs, B=@n2.UVs, Alpha=@n3), B=AppendVector(A=@n4.OffsetFlechaX, B=@n5))
   Time=Multiply(A=Time(), B=ScalarParameter() {defaultValue=1, parameterName=TimeMultiplier})
   Speed=AppendVector(A=ScalarParameter() {defaultValue=-1, parameterName=SpéedX}, B=ScalarParameter() {parameterName=SpéedY})
  ) {speedX=-1}
  n7 = TextureSampleParameter2D!(UVs=@n6, Level, Apply View MipBias) {
   automaticViewMipBias=false
   mipValueMode=TMVM_MipLevel
   parameterName=MaterialExpressionTextureSampleParameter2D_1
   texture=/VehicleCosmetics/DriftTrail/DriftTrail_GPS/Textures/T_VCVFX_RibbonGPSWT_RGB
  }
  n8 = LinearInterpolate(A=Desaturation(None=@n7.RGB), B=@n7.RGB, Alpha=ScalarParameter() {defaultValue=1, desc="1=X3Y1; 0=X1Y1", parameterName=SelectorTextCoord})
  n9 = ParticleColor()
  n10 = DynamicParameter() {parameterIndex=2, defaultValue=<rgba 0.8469 0.521 0.008 1>, paramNames [RSmallLine,GSmallLine,BSmallLine,ASmallLine]}
  n11 = LinearInterpolate(
   A=DynamicParameter() {defaultValue=<rgba 1 0.5 0.1 1>, paramNames [RMediumLine,GMediumLine,BMediumLine,AMediumLine]}.RGB
   B=@n10.RGB
   Alpha=@n7.R
  )
  n12 = DynamicParameter() {parameterIndex=3, defaultValue=<rgba 0.8469 0.521 0.008 1>, paramNames [RBackLine,GBackLine,BBackLine,ABackLine]}
  MaterialFunctionCall(Emissive=LinearInterpolate(
    A=Multiply(A=@n8, B=@n9.RGB)
    B=LinearInterpolate(A=@n11, B=@n12.RGB, Alpha=@n7.B)
    Alpha=ScalarParameter() {defaultValue=1, desc="1=X3Y1; 0=X1Y1", parameterName=SelectorTextCoord}
   )) {materialFunction=&MF_FX_EmissiveAdjustments}
  n13 = ParticleColor()
  n14 = DynamicParameter() {defaultValue=<rgba 1 0.5 0.1 1>, paramNames [RMediumLine,GMediumLine,BMediumLine,AMediumLine]}
  n15 = TextureCoordinate()
  n16 = TextureCoordinate() {uTiling=3}
  n17 = ScalarParameter() {defaultValue=1, desc="1=X3Y1; 0=X1Y1", parameterName=SelectorTextCoord}
  n18 = DynamicParameter() {parameterIndex=1, defaultValue=<rgba 0 1 1 1>, paramNames [IsPainted?, OffsetFlechaX, CentralLineMulti, Switch_TyreIntermit]}
  n19 = ScalarParameter() {parameterName=OffsetFlechaY}
  n20 = Panner(
   Coordinate=Add(A=LinearInterpolate(A=@n15, B=@n16, Alpha=@n17), B=AppendVector(A=@n18.OffsetFlechaX, B=@n19))
   Time=Multiply(A=Time(), B=ScalarParameter() {defaultValue=1, parameterName=TimeMultiplier})
   Speed=AppendVector(A=ScalarParameter() {defaultValue=-1, parameterName=SpéedX}, B=ScalarParameter() {parameterName=SpéedY})
  ) {speedX=-1}
  n21 = TextureSampleParameter2D!(UVs=@n20, Apply View MipBias) {
   parameterName=MaterialExpressionTextureSampleParameter2D_1
   samplerSource=SSM_Wrap_WorldGroupSettings
   texture=&T_VCVFX_RibbonGPSWT_RGB
  }
  n22 = Multiply(A=@n14.GMediumLine, B=@n21.R)
  n23 = Multiply(A=@n14.BMediumLine, B=@n21.G)
  n24 = TextureSampleParameter2D(UVs=@n20) {
   parameterName=MaterialExpressionTextureSampleParameter2D_0
   samplerSource=SSM_Wrap_WorldGroupSettings
   texture=/VehicleCosmetics/DriftTrail/DriftTrail_GPS/Textures/T_VCVFX_RibbonGPS_RGB
  }
  n25 = LinearInterpolate(A=Add(A=Add(A=@n22, B=@n23), B=Multiply(A=@n14.AMediumLine, B=@n21.B)), B=@n24.RGB, Alpha=@n14.RMediumLine)
  n26 = DynamicParameter() {parameterIndex=1, defaultValue=<rgba 0 1 1 1>, paramNames [IsPainted?, OffsetFlechaX, CentralLineMulti, Switch_TyreIntermit]}
  MaterialExpressionTextureSampleParameter2D_0 = TextureSampleParameter2D() {
   parameterName=MaterialExpressionTextureSampleParameter2D_0
   samplerSource=SSM_Wrap_WorldGroupSettings
   texture=&T_VCVFX_RibbonGPS_RGB
  }
  n27 = ComponentMask(None=TextureCoordinate()) {g=true}
  n28 = Round(None=Power(Base=Sine(None=@n27) {period=2}) {constExponent=6.5})
  n29 = Multiply(A=@MaterialExpressionTextureSampleParameter2D_0.RGB, B=@n28)
  n30 = OneMinus(None=@n28)
  n31 = Add(
   A=Multiply(A=@n26.CentralLineMulti, B=Desaturation(None=@n29))
   B=Desaturation(None=Multiply(A=@MaterialExpressionTextureSampleParameter2D_0.RGB, B=@n30))
  )
  MaterialFunctionCall(Emissive=Multiply(A=@n13.RGB, B=LinearInterpolate(A=@n25, B=@n31, Alpha=@n26.IsPainted?))) {materialFunction=&MF_FX_EmissiveAdjustments}
  n32 = TextureCoordinate()
  n33 = Power(Base=OneMinus(None=ComponentMask(None=@n32) {r=true}), Exp=ScalarParameter() {defaultValue=3, parameterName=PowerNegativeMask})
  n34 = Constant() {r=1}
  n35 = Power(Base=ComponentMask(None=@n32) {r=true}, Exp=ScalarParameter() {parameterName=PowerPositiveMask})
  n36 = Constant() {r=500}
  MaterialExpressionConstantClamp_2 = Saturate(None=Multiply(A=Multiply(A=@n33, B=@n34), B=Multiply(A=@n35, B=@n36)))
  n37 = ScalarParameter() {parameterName="Trail=0;Wheel=1"}
  Multiply(A=@n13.A, B=Multiply(A=@n24.A, B=LinearInterpolate(A=@MaterialExpressionConstantClamp_2, Alpha=@n37)))
  MaterialExpressionTextureSampleParameter2D_1 = TextureSampleParameter2D!(UVs=@n6, Level, Apply View MipBias) {
   automaticViewMipBias=false
   mipValueMode=TMVM_MipLevel
   parameterName=MaterialExpressionTextureSampleParameter2D_0
   texture=&T_VCVFX_RibbonGPS_RGB
  }
  n38 = LinearInterpolate(
   A=@MaterialExpressionTextureSampleParameter2D_1.A
   B=Add(A=@MaterialExpressionTextureSampleParameter2D_1.G, B=@MaterialExpressionTextureSampleParameter2D_1.G)
   Alpha=ScalarParameter() {parameterName="Trail=0;Wheel=1"}
  ) {constAlpha=1}
  n39 = ScalarParameter() {parameterName=PowerGrecas}
  n40 = TextureCoordinate()
  n41 = Power(Base=OneMinus(None=ComponentMask(None=@n40) {r=true}), Exp=ScalarParameter() {defaultValue=3, parameterName=PowerNegativeMask})
  n42 = Constant() {r=1}
  n43 = Power(Base=ComponentMask(None=@n40) {r=true}, Exp=ScalarParameter() {parameterName=PowerPositiveMask})
  n44 = Constant() {r=500}
  n45 = Saturate(None=Multiply(A=Multiply(A=@n41, B=@n42), B=Multiply(A=@n43, B=@n44)))
  n46 = ScalarParameter() {parameterName="Trail=0;Wheel=1"}
  n47 = Multiply(A=@n9.A, B=Multiply(A=Power(Base=@n38, Exp=@n39), B=LinearInterpolate(A=@n45, Alpha=@n46)))
  n48 = ComponentMask(None=TextureCoordinate()) {g=true}
  n49 = Round(None=Power(Base=Sine(None=@n48) {period=2}, Exp=ScalarParameter() {defaultValue=6, parameterName=MaskRoundWheel}) {constExponent=6})
  n50 = DynamicParameter() {parameterIndex=1, defaultValue=<rgba 0 1 1 1>, paramNames [IsPainted?, OffsetFlechaX, CentralLineMulti, Switch_TyreIntermit]}
  Multiply(
   A=LinearInterpolate(A=@n47, B=Multiply(A=@n47, B=@n49), Alpha=ScalarParameter() {parameterName="Trail=0;Wheel=1"}) {constAlpha=1}
   B=@n50.Switch_TyreIntermit
  )
 ]
}
